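# Animation Blueprint for the Renegade Skull backpack that adds subtle springy secondary motion to the pack bone.
# Two chained Anim Dynamics linear springs start from the ref pose, allowing X/Z translation with Y locked and scaled-down alpha.
# Each update, bIsInFrontEnd sets complementary 0/1 alphas, selecting a stiffer, less motion-sensitive spring in the front end and a looser one elsewhere.
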AnimBlueprint /BRCosmetics/Accessories/FORT_Backpacks/Backpack_F_MED_Renegade_Skull/Meshes/F_MED_Renegade_Skull_Pack_AnimBP
target_skeleton=&/F_MED_Renegade_Skull_Pack_Skeleton
graphs [
 AnimGraph {nodes [
   n1 = Anim Dynamics - Bone: pack(
    ComponentPose=Local To Component(LocalPose=Local Space Ref Pose())
    Alpha=Get isNotFrontEnd() -> (isNotFrontEnd:double)
    ExternalForce="0,0,0"
   ) {
    bGravityOverrideInSimSpace=true
    bLinearSpring=true
    gravityScale=0.1
    linearSpringConstant=20
    alphaBoolBlend {blendOption=Linear}
    alphaScaleBiasClamp {scale=0.1}
    boundBone {boneName=pack}
    componentLinearAccScale=<0.35 0.2 0.15>
    componentLinearVelScale=<0.35 0.2 0.15>
    gravityOverride {z=-980}
    physicsBodyDefinitions [{
      sphereCollisionRadius=0
      boundBone {boneName=pack}
      boxExtents=<1 1 1>
      constraintSetup {angularLimitsMax=<0 0 0>, angularLimitsMin=<0 0 0>, angularTarget=<0 0 0>, linearAxesMax=<500 0 500>, linearAxesMin=<-500 0 -500>}
      localJointOffset=<0 0 0>
     }]
   }
   n2 = Anim Dynamics - Bone: pack(ComponentPose=@n1, Alpha=Get isFrontEnd() -> (isFrontEnd:double), ExternalForce="0,0,0") {
    bGravityOverrideInSimSpace=true
    bLinearSpring=true
    gravityScale=0.1
    linearSpringConstant=50
    alphaBoolBlend {blendOption=Linear}
    alphaScaleBiasClamp {scale=0.1}
    boundBone {boneName=pack}
    componentAppliedLinearAccClamp=<10000 10000 10000>
    componentLinearAccScale=<0.05 0.05 0.15>
    componentLinearVelScale=<0.05 0.05 0.15>
    gravityOverride {z=-980}
    physicsBodyDefinitions [{
      sphereCollisionRadius=0
      boundBone {boneName=pack}
      boxExtents=<1 1 1>
      constraintSetup {angularLimitsMax=<0 0 0>, angularLimitsMin=<0 0 0>, angularTarget=<0 0 0>, linearAxesMax=<500 0 500>, linearAxesMin=<-500 0 -500>}
      localJointOffset=<0 0 0>
     }]
   }
   Output Pose(Result=Component To Local(ComponentPose=@n2)) {layerGroup=None}
  ]}
 EventGraph {nodes [
   Event BlueprintUpdateAnimation()
   > n1 = Branch(Condition=Get bIsInFrontEnd())
   > Set isFrontEnd(isFrontEnd:double=1) -> (Output_Get:double)
   > Set isNotFrontEnd(isNotFrontEnd:double=0) -> (Output_Get:double)
   Set isFrontEnd(execute=@n1.else, isFrontEnd=0)
   > Set isNotFrontEnd(isNotFrontEnd=1)
  ]}
]
variables [
 backpack rotate {type=Rotator}
 backpack translate {type=number}
 isFrontEnd {type=number}
 isNotFrontEnd {type=number}
]
